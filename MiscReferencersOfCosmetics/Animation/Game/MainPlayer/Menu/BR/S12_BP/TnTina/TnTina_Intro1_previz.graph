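# Previz Level Sequence for TnTina's first intro in the Season 12 Battle Pass menu, set among oil-rig props.
# Keyframed cine cameras cover successive shots, with camera shake late in the sequence.
# A female mannequin animates in step with chair, grenade box, grenades, pins and hoop gadgets, alongside a timed Niagara effect.
# Events fire on two light switchers, and set dressing is hidden or jumped aside mid-sequence.
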LevelSequence /MiscReferencersOfCosmetics/Animation/Game/MainPlayer/Menu/BR/S12_BP/TnTina/TnTina_Intro1_previz
bindings [
 BP_Battlepass_LightSwitcher_Maproom {
  id=#1
  possessed_class=BP_Battlepass_LightSwitcher_C
  tracks [{
    type=MovieSceneEventTrack
    sections [{
      type=MovieSceneEventTriggerSection
      channels [None {type=MovieSceneScriptingEventChannel, const ["<Struct 'MovieSceneEvent' (0x851953b5a00) {}>", 1]}]
     }]
   }]
 }
 CameraComponent {
  id=#2
  parent_id=#3
  possessed_class=CineCameraComponent
  tracks [
   {
    type=MovieSceneFloatTrack
    display_name=Current Focal Length
    sections [{type=MovieSceneFloatSection, channels [None {type=MovieSceneScriptingFloatChannel, default=18}]}]
   }
   {
    type=MovieSceneFloatTrack
    display_name="Manual Focus Distance (Focus Settings)"
    sections [{type=MovieSceneFloatSection, channels [None {type=MovieSceneScriptingFloatChannel, default=14.168}]}]
   }
   {
    type=MovieSceneFloatTrack
    display_name=Current Aperture
    sections [{type=MovieSceneFloatSection, channels [None {type=MovieSceneScriptingFloatChannel, default=10.2533}]}]
   }
  ]
 }
 S_OilRig_BatteryJumper_01 {
  id=#4
  possessed_class=FortStaticMeshActor
  tracks [
   {
    type=MovieScene3DTransformTrack
    sections [{
      type=MovieScene3DTransformSection
      channels [
       Location.X {type=MovieSceneScriptingDoubleChannel, default=2164}
       Location.Y {type=MovieSceneScriptingDoubleChannel, default=11059}
       Location.Z {type=MovieSceneScriptingDoubleChannel, default=780}
       Rotation.X {type=MovieSceneScriptingDoubleChannel, default=0}
       Rotation.Y {type=MovieSceneScriptingDoubleChannel, default=0}
       Rotation.Z {type=MovieSceneScriptingDoubleChannel, default=-150.0001}
       Scale.X {type=MovieSceneScriptingDoubleChannel, default=1}
       Scale.Y {type=MovieSceneScriptingDoubleChannel, default=1}
       Scale.Z {type=MovieSceneScriptingDoubleChannel, default=1}
      ]
     }]
   }
   {
    type=MovieSceneVisibilityTrack
    sections [{type=MovieSceneVisibilitySection, channels [None {type=MovieSceneScriptingBoolChannel, default=false}]}]
   }
  ]
 }
 S_OilRig_DeskChair_01 {
  id=#5
  possessed_class=FortStaticMeshActor
  tracks [
   {
    type=MovieScene3DTransformTrack
    sections [{
      type=MovieScene3DTransformSection
      channels [
       Location.X {type=MovieSceneScriptingDoubleChannel, default=2161.4993}
       Location.Y {type=MovieSceneScriptingDoubleChannel, default=11143.5967}
       Location.Z {type=MovieSceneScriptingDoubleChannel, default=781.1124}
       Rotation.X {type=MovieSceneScriptingDoubleChannel, default=0}
       Rotation.Y {type=MovieSceneScriptingDoubleChannel, default=0}
       Rotation.Z {type=MovieSceneScriptingDoubleChannel, default=140.1271}
       Scale.X {type=MovieSceneScriptingDoubleChannel, default=1}
       Scale.Y {type=MovieSceneScriptingDoubleChannel, default=1}
       Scale.Z {type=MovieSceneScriptingDoubleChannel, default=1}
      ]
     }]
   }
   {type=MovieSceneSpawnTrack, sections [{type=MovieSceneSpawnSection, channels [None {type=MovieSceneScriptingBoolChannel, default=false}]}]}
  ]
 }
 SHOT_0020 {
  id=#3
  possessed_class=CineCameraActor
  tracks [
   {
    type=MovieScene3DTransformTrack
    sections [{
      type=MovieScene3DTransformSection
      channels [
       Location.X {type=MovieSceneScriptingDoubleChannel, default=2330.8176, const [2299.3127,3]}
       Location.Y {type=MovieSceneScriptingDoubleChannel, default=11073.9375, thinned_from=3, keys [[118,11136.1367],[150,11140.918]]}
       Location.Z {type=MovieSceneScriptingDoubleChannel, default=972.313, const [980.8381,3]}
       Rotation.X {type=MovieSceneScriptingDoubleChannel, default=0, thinned_from=3, keys [[118,4.2942],[150,4.2941]]}
       Rotation.Y {type=MovieSceneScriptingDoubleChannel, default=15.9864, keys [[118,-4.5779],[135,17.7653],[150,22.9084]]}
       Rotation.Z {type=MovieSceneScriptingDoubleChannel, default=-202.4009, keys [[118,-156.7864],[135,-156.3861],[150,-157.5863]]}
       Scale.X {type=MovieSceneScriptingDoubleChannel, default=1, const [1,3]}
       Scale.Y {type=MovieSceneScriptingDoubleChannel, default=1, const [1,3]}
       Scale.Z {type=MovieSceneScriptingDoubleChannel, default=1, const [1,3]}
      ]
     }]
   }
   {type=MovieSceneSpawnTrack, sections [{type=MovieSceneSpawnSection, channels [None {type=MovieSceneScriptingBoolChannel, default=true}]}]}
  ]
 }
 SHOT_0010 {
  id=#6
  possessed_class=CineCameraActor
  tracks [
   {
    type=MovieScene3DTransformTrack
    sections [{
      type=MovieScene3DTransformSection
      channels [
       Location.X {type=MovieSceneScriptingDoubleChannel, default=2163.0708, thinned_from=3, keys [[1,2235.1814],[69,2217.7439]]}
       Location.Y {type=MovieSceneScriptingDoubleChannel, default=10933.375, keys [[1,11005],[69,10890]]}
       Location.Z {type=MovieSceneScriptingDoubleChannel, default=795.0637, const [791,3]}
       Rotation.X {type=MovieSceneScriptingDoubleChannel, default=0, const [4.4685,2]}
       Rotation.Y {type=MovieSceneScriptingDoubleChannel, default=22.8447, const [23.375,2]}
       Rotation.Z {type=MovieSceneScriptingDoubleChannel, default=164.215, thinned_from=3, keys [[1,188.2157],[69,161.7779]]}
       Scale.X {type=MovieSceneScriptingDoubleChannel, default=1, const [1,3]}
       Scale.Y {type=MovieSceneScriptingDoubleChannel, default=1, const [1,3]}
       Scale.Z {type=MovieSceneScriptingDoubleChannel, default=1, const [1,3]}
      ]
     }]
   }
   {type=MovieSceneSpawnTrack, sections [{type=MovieSceneSpawnSection, channels [None {type=MovieSceneScriptingBoolChannel, default=true}]}]}
  ]
 }
 SoldierMannequin_Female2 {
  id=#7
  possessed_class=SoldierMannequin_Female_C
  tracks [
   {
    type=MovieScene3DTransformTrack
    sections [{
      type=MovieScene3DTransformSection
      channels [
       Location.X {type=MovieSceneScriptingDoubleChannel, default=-810.7837, const [0.2132,1]}
       Location.Y {type=MovieSceneScriptingDoubleChannel, default=204.6233, const [0,1]}
       Location.Z {type=MovieSceneScriptingDoubleChannel, default=0, const [0,1]}
       Rotation.X {type=MovieSceneScriptingDoubleChannel, default=0, const [0,1]}
       Rotation.Y {type=MovieSceneScriptingDoubleChannel, default=0, const [0,1]}
       Rotation.Z {type=MovieSceneScriptingDoubleChannel, default=-4.2561, const [0,1]}
       Scale.X {type=MovieSceneScriptingDoubleChannel, default=1, const [1,1]}
       Scale.Y {type=MovieSceneScriptingDoubleChannel, default=1, const [1,1]}
       Scale.Z {type=MovieSceneScriptingDoubleChannel, default=1, const [1,1]}
      ]
     }]
   }
   {
    type=MovieSceneSkeletalAnimationTrack
    sections [
     {type=MovieSceneSkeletalAnimationSection, end_frame=229, start_frame=185}
     {type=MovieSceneSkeletalAnimationSection, end_frame=52, start_frame=1}
     {type=MovieSceneSkeletalAnimationSection, end_frame=120, start_frame=91}
     {type=MovieSceneSkeletalAnimationSection, end_frame=91, start_frame=52}
    ]
   }
   {
    type=MovieSceneSkeletalAnimationTrack
    sections [
     {type=MovieSceneSkeletalAnimationSection, end_frame=253, start_frame=229}
     {type=MovieSceneSkeletalAnimationSection, end_frame=364, start_frame=253}
     {type=MovieSceneSkeletalAnimationSection, end_frame=491, start_frame=414}
    ]
   }
   {
    type=MovieSceneVisibilityTrack
    sections [{type=MovieSceneVisibilitySection, channels [None {type=MovieSceneScriptingBoolChannel, default=true}]}]
   }
  ]
 }
 Shot_50_Alt {
  id=#8
  possessed_class=CineCameraActor
  tracks [{
    type=MovieScene3DTransformTrack
    sections [{
      type=MovieScene3DTransformSection
      channels [
       Location.X {type=MovieSceneScriptingDoubleChannel, default=0, const [2376.2046,1]}
       Location.Y {type=MovieSceneScriptingDoubleChannel, default=0, const [11068.6445,1]}
       Location.Z {type=MovieSceneScriptingDoubleChannel, default=0, const [781.231,1]}
       Rotation.X {type=MovieSceneScriptingDoubleChannel, default=0, const [-0.5591,1]}
       Rotation.Y {type=MovieSceneScriptingDoubleChannel, default=0, const [22.4169,1]}
       Rotation.Z {type=MovieSceneScriptingDoubleChannel, default=0, const [152.0269,1]}
       Scale.X {type=MovieSceneScriptingDoubleChannel, default=1, const [1,1]}
       Scale.Y {type=MovieSceneScriptingDoubleChannel, default=1, const [1,1]}
       Scale.Z {type=MovieSceneScriptingDoubleChannel, default=1, const [1,1]}
      ]
     }]
   }]
 }
 S_OilRig_TrashCan2 {
  id=#9
  possessed_class=FortStaticMeshActor
  tracks [{
    type=MovieScene3DTransformTrack
    sections [{
      type=MovieScene3DTransformSection
      channels [
       Location.X {type=MovieSceneScriptingDoubleChannel, default=2363.7939, const [2363.7939,1]}
       Location.Y {type=MovieSceneScriptingDoubleChannel, default=11257.4209, const [11257.4209,1]}
       Location.Z {type=MovieSceneScriptingDoubleChannel, default=780, const [780,1]}
       Rotation.X {type=MovieSceneScriptingDoubleChannel, default=0, const [0,1]}
       Rotation.Y {type=MovieSceneScriptingDoubleChannel, default=0, const [0,1]}
       Rotation.Z {type=MovieSceneScriptingDoubleChannel, default=171.5623, const [171.5623,1]}
       Scale.X {type=MovieSceneScriptingDoubleChannel, default=1.1962, const [1.1962,1]}
       Scale.Y {type=MovieSceneScriptingDoubleChannel, default=1.1962, const [1.1962,1]}
       Scale.Z {type=MovieSceneScriptingDoubleChannel, default=0.8651, const [0.8651,1]}
      ]
     }]
   }]
 }
 SkeletalMeshComponent0 {
  id=#10
  parent_id=#7
  possessed_class=SkeletalMeshComponent
  tracks [{
    type=MovieSceneFloatTrack
    display_name=Bounds Scale
    sections [{type=MovieSceneFloatSection, channels [None {type=MovieSceneScriptingFloatChannel, default=50000, const [50000,1]}]}]
   }]
 }
 CameraComponent {
  id=#11
  parent_id=#12
  possessed_class=CineCameraComponent
  tracks [
   {
    type=MovieSceneFloatTrack
    display_name=Current Focal Length
    sections [{type=MovieSceneFloatSection, channels [None {type=MovieSceneScriptingFloatChannel, default=25}]}]
   }
   {
    type=MovieSceneFloatTrack
    display_name="Manual Focus Distance (Focus Settings)"
    sections [{type=MovieSceneFloatSection, channels [None {type=MovieSceneScriptingFloatChannel, default=60}]}]
   }
   {
    type=MovieSceneFloatTrack
    display_name=Current Aperture
    sections [{type=MovieSceneFloatSection, channels [None {type=MovieSceneScriptingFloatChannel, default=5.5072}]}]
   }
  ]
 }
 CameraComponent {
  id=#13
  parent_id=#14
  possessed_class=CineCameraComponent
  tracks [
   {
    type=MovieSceneFloatTrack
    display_name=Current Focal Length
    sections [{type=MovieSceneFloatSection, channels [None {type=MovieSceneScriptingFloatChannel, default=35}]}]
   }
   {
    type=MovieSceneFloatTrack
    display_name="Manual Focus Distance (Focus Settings)"
    sections [{type=MovieSceneFloatSection, channels [None {type=MovieSceneScriptingFloatChannel, default=100000}]}]
   }
   {
    type=MovieSceneFloatTrack
    display_name=Current Aperture
    sections [{type=MovieSceneFloatSection, channels [None {type=MovieSceneScriptingFloatChannel, default=22.3406}]}]
   }
  ]
 }
 NiagaraComponent0 {
  id=#15
  parent_id=#16
  possessed_class=NiagaraComponent
  tracks [{type=MovieSceneNiagaraSystemTrack, sections [{type=MovieSceneNiagaraSystemSpawnSection, end_frame=364, start_frame=285}]}]
 }
 S_OilRig_Device_01 {
  id=#17
  possessed_class=FortStaticMeshActor
  tracks [
   {
    type=MovieScene3DTransformTrack
    sections [{
      type=MovieScene3DTransformSection
      channels [
       Location.X {type=MovieSceneScriptingDoubleChannel, default=2076}
       Location.Y {type=MovieSceneScriptingDoubleChannel, default=11031.2871}
       Location.Z {type=MovieSceneScriptingDoubleChannel, default=804}
       Rotation.X {type=MovieSceneScriptingDoubleChannel, default=0}
       Rotation.Y {type=MovieSceneScriptingDoubleChannel, default=0}
       Rotation.Z {type=MovieSceneScriptingDoubleChannel, default=-26.2497}
       Scale.X {type=MovieSceneScriptingDoubleChannel, default=-1}
       Scale.Y {type=MovieSceneScriptingDoubleChannel, default=1}
       Scale.Z {type=MovieSceneScriptingDoubleChannel, default=1}
      ]
     }]
   }
   {
    type=MovieSceneVisibilityTrack
    sections [{type=MovieSceneVisibilitySection, channels [None {type=MovieSceneScriptingBoolChannel, default=false}]}]
   }
  ]
 }
 Shot_0045 {
  id=#14
  possessed_class=CineCameraActor
  tracks [
   {
    type=MovieScene3DTransformTrack
    sections [{
      type=MovieScene3DTransformSection
      channels [
       Location.X {type=MovieSceneScriptingDoubleChannel, default=2018.4292, const [2018.4292,2]}
       Location.Y {type=MovieSceneScriptingDoubleChannel, default=10875.3076, const [10875.3076,2]}
       Location.Z {type=MovieSceneScriptingDoubleChannel, default=862.3153, const [862.3153,2]}
       Rotation.X {type=MovieSceneScriptingDoubleChannel, default=-0.8426, const [-0.8426,2]}
       Rotation.Y {type=MovieSceneScriptingDoubleChannel, default=0.1545, const [0.1545,2]}
       Rotation.Z {type=MovieSceneScriptingDoubleChannel, default=109.7859, keys [[243.0063,109.7859],[253,98.8007]]}
       Scale.X {type=MovieSceneScriptingDoubleChannel, default=1, const [1,2]}
       Scale.Y {type=MovieSceneScriptingDoubleChannel, default=1, const [1,2]}
       Scale.Z {type=MovieSceneScriptingDoubleChannel, default=1, const [1,2]}
      ]
     }]
   }
   {type=MovieSceneSpawnTrack, sections [{type=MovieSceneSpawnSection, channels [None {type=MovieSceneScriptingBoolChannel, default=true}]}]}
  ]
 }
 Shot_0040 {
  id=#18
  possessed_class=CineCameraActor
  tracks [
   {
    type=MovieScene3DTransformTrack
    sections [{
      type=MovieScene3DTransformSection
      channels [
       Location.X {type=MovieSceneScriptingDoubleChannel, default=2125.0149, const [2002.394,1]}
       Location.Y {type=MovieSceneScriptingDoubleChannel, default=10982.3916, const [10937.7793,1]}
       Location.Z {type=MovieSceneScriptingDoubleChannel, default=881.7528, const [894.1627,1]}
       Rotation.X {type=MovieSceneScriptingDoubleChannel, default=0, const [-4.0118,1]}
       Rotation.Y {type=MovieSceneScriptingDoubleChannel, default=5.4835, const [-0.9423,1]}
       Rotation.Z {type=MovieSceneScriptingDoubleChannel, default=168.1625, const [48.2748,1]}
       Scale.X {type=MovieSceneScriptingDoubleChannel, default=1, const [1,1]}
       Scale.Y {type=MovieSceneScriptingDoubleChannel, default=1, const [1,1]}
       Scale.Z {type=MovieSceneScriptingDoubleChannel, default=1, const [1,1]}
      ]
     }]
   }
   {type=MovieSceneSpawnTrack, sections [{type=MovieSceneSpawnSection, channels [None {type=MovieSceneScriptingBoolChannel, default=true}]}]}
  ]
 }
 Grenade_pin2 {
  id=#19
  possessed_class=SkeletalMeshActor
  tracks [
   {
    type=MovieScene3DTransformTrack
    sections [{
      type=MovieScene3DTransformSection
      channels [
       Location.X {type=MovieSceneScriptingDoubleChannel, default=0}
       Location.Y {type=MovieSceneScriptingDoubleChannel, default=0}
       Location.Z {type=MovieSceneScriptingDoubleChannel, default=0}
       Rotation.X {type=MovieSceneScriptingDoubleChannel, default=0}
       Rotation.Y {type=MovieSceneScriptingDoubleChannel, default=0}
       Rotation.Z {type=MovieSceneScriptingDoubleChannel, default=0}
       Scale.X {type=MovieSceneScriptingDoubleChannel, default=1}
       Scale.Y {type=MovieSceneScriptingDoubleChannel, default=1}
       Scale.Z {type=MovieSceneScriptingDoubleChannel, default=1}
      ]
     }]
   }
   {type=MovieSceneSpawnTrack, sections [{type=MovieSceneSpawnSection, channels [None {type=MovieSceneScriptingBoolChannel, default=true}]}]}
   {type=MovieSceneSkeletalAnimationTrack, sections [{type=MovieSceneSkeletalAnimationSection, end_frame=226, start_frame=185}]}
   {type=MovieSceneSkeletalAnimationTrack, sections [{type=MovieSceneSkeletalAnimationSection, end_frame=380, start_frame=253}]}
  ]
 }
 SHOT_0016 {
  id=#20
  possessed_class=CineCameraActor
  tracks [
   {
    type=MovieScene3DTransformTrack
    sections [{
      type=MovieScene3DTransformSection
      channels [
       Location.X {type=MovieSceneScriptingDoubleChannel, default=2130.2444, const [2123.8936,4]}
       Location.Y {type=MovieSceneScriptingDoubleChannel, default=10948.3184, const [10930.0469,4]}
       Location.Z {type=MovieSceneScriptingDoubleChannel, default=870.4947, const [857.192,4]}
       Rotation.X {type=MovieSceneScriptingDoubleChannel, default=0.0001, thinned_from=4, keys [[91,0.0001],[109,-0.2043],[121,-0.2043]]}
       Rotation.Y {type=MovieSceneScriptingDoubleChannel, default=8.2047, thinned_from=4, keys [[91,19.8046],[109,19.787],[121,25.7867]]}
       Rotation.Z {type=MovieSceneScriptingDoubleChannel, default=2686.1787, thinned_from=4, keys [[91,2683.5767],[115,2677.8655],[121,2671.2581]]}
       Scale.X {type=MovieSceneScriptingDoubleChannel, default=1, const [1,4]}
       Scale.Y {type=MovieSceneScriptingDoubleChannel, default=1, const [1,4]}
       Scale.Z {type=MovieSceneScriptingDoubleChannel, default=1, const [1,4]}
      ]
     }]
   }
   {type=MovieSceneSpawnTrack, sections [{type=MovieSceneSpawnSection, channels [None {type=MovieSceneScriptingBoolChannel, default=true}]}]}
  ]
 }
 "S_OilRig_DeskChair_01 (2)" {
  id=#21
  possessed_class=FortStaticMeshActor
  tracks [
   {
    type=MovieScene3DTransformTrack
    sections [{
      type=MovieScene3DTransformSection
      channels [
       Location.X {type=MovieSceneScriptingDoubleChannel, default=2008}
       Location.Y {type=MovieSceneScriptingDoubleChannel, default=10880}
       Location.Z {type=MovieSceneScriptingDoubleChannel, default=780}
       Rotation.X {type=MovieSceneScriptingDoubleChannel, default=0}
       Rotation.Y {type=MovieSceneScriptingDoubleChannel, default=0}
       Rotation.Z {type=MovieSceneScriptingDoubleChannel, default=129.375}
       Scale.X {type=MovieSceneScriptingDoubleChannel, default=1}
       Scale.Y {type=MovieSceneScriptingDoubleChannel, default=1}
       Scale.Z {type=MovieSceneScriptingDoubleChannel, default=1}
      ]
     }]
   }
   {type=MovieSceneSpawnTrack, sections [{type=MovieSceneSpawnSection, channels [None {type=MovieSceneScriptingBoolChannel, default=false}]}]}
  ]
 }
 BP_Battlepass_LightSwitcher_TNTINA {
  id=#22
  possessed_class=BP_Battlepass_LightSwitcher_C
  tracks [{
    type=MovieSceneEventTrack
    sections [{
      type=MovieSceneEventTriggerSection
      channels [None {type=MovieSceneScriptingEventChannel, const ["<Struct 'MovieSceneEvent' (0x851953b55a0) {}>", 1]}]
     }]
   }]
 }
 CameraComponent {
  id=#23
  parent_id=#24
  possessed_class=CameraComponent
  tracks [{
    type=MovieSceneFloatTrack
    display_name=Field Of View
    sections [{type=MovieSceneFloatSection, channels [None {type=MovieSceneScriptingFloatChannel, default=80, const [80,1]}]}]
   }]
 }
 TNTina_Chair_Gadget {
  id=#25
  possessed_class=SkeletalMeshActor
  tracks [
   {
    type=MovieScene3DTransformTrack
    sections [{
      type=MovieScene3DTransformSection
      channels [
       Location.X {type=MovieSceneScriptingDoubleChannel, default=0}
       Location.Y {type=MovieSceneScriptingDoubleChannel, default=0}
       Location.Z {type=MovieSceneScriptingDoubleChannel, default=0}
       Rotation.X {type=MovieSceneScriptingDoubleChannel, default=0}
       Rotation.Y {type=MovieSceneScriptingDoubleChannel, default=0}
       Rotation.Z {type=MovieSceneScriptingDoubleChannel, default=90}
       Scale.X {type=MovieSceneScriptingDoubleChannel, default=1}
       Scale.Y {type=MovieSceneScriptingDoubleChannel, default=1}
       Scale.Z {type=MovieSceneScriptingDoubleChannel, default=1}
      ]
     }]
   }
   {type=MovieSceneSkeletalAnimationTrack, sections []}
   {
    type=MovieSceneSkeletalAnimationTrack
    sections [
     {type=MovieSceneSkeletalAnimationSection, end_frame=90, start_frame=52}
     {type=MovieSceneSkeletalAnimationSection, end_frame=251, start_frame=227}
     {type=MovieSceneSkeletalAnimationSection, end_frame=52, start_frame=1}
     {type=MovieSceneSkeletalAnimationSection, end_frame=120, start_frame=90}
     {type=MovieSceneSkeletalAnimationSection, end_frame=369, start_frame=253}
    ]
   }
   {type=MovieSceneSkeletalAnimationTrack, sections []}
   {type=MovieSceneSkeletalAnimationTrack, sections [{type=MovieSceneSkeletalAnimationSection, end_frame=229, start_frame=184}]}
   {
    type=MovieSceneVisibilityTrack
    sections [{type=MovieSceneVisibilitySection, channels [None {type=MovieSceneScriptingBoolChannel, default=true}]}]
   }
  ]
 }
 TNTina_Grenade_Box_Gadget {
  id=#26
  possessed_class=SkeletalMeshActor
  tracks [
   {
    type=MovieScene3DTransformTrack
    sections [{
      type=MovieScene3DTransformSection
      channels [
       Location.X {type=MovieSceneScriptingDoubleChannel, default=0}
       Location.Y {type=MovieSceneScriptingDoubleChannel, default=0}
       Location.Z {type=MovieSceneScriptingDoubleChannel, default=0}
       Rotation.X {type=MovieSceneScriptingDoubleChannel, default=0}
       Rotation.Y {type=MovieSceneScriptingDoubleChannel, default=0}
       Rotation.Z {type=MovieSceneScriptingDoubleChannel, default=90}
       Scale.X {type=MovieSceneScriptingDoubleChannel, default=1}
       Scale.Y {type=MovieSceneScriptingDoubleChannel, default=1}
       Scale.Z {type=MovieSceneScriptingDoubleChannel, default=1}
      ]
     }]
   }
   {type=MovieSceneSkeletalAnimationTrack, sections []}
   {type=MovieSceneSkeletalAnimationTrack, sections []}
   {
    type=MovieSceneSkeletalAnimationTrack
    sections [
     {type=MovieSceneSkeletalAnimationSection, end_frame=185, start_frame=91}
     {type=MovieSceneSkeletalAnimationSection, end_frame=364, start_frame=236}
     {type=MovieSceneSkeletalAnimationSection, end_frame=230, start_frame=185}
    ]
   }
   {
    type=MovieSceneVisibilityTrack
    sections [{type=MovieSceneVisibilitySection, channels [None {type=MovieSceneScriptingBoolChannel, default=true}]}]
   }
  ]
 }
 SHOT_0015 {
  id=#27
  possessed_class=CineCameraActor
  tracks [
   {
    type=MovieScene3DTransformTrack
    sections [{
      type=MovieScene3DTransformSection
      channels [
       Location.X {type=MovieSceneScriptingDoubleChannel, default=1993.4319, const [2007.5182,3]}
       Location.Y {type=MovieSceneScriptingDoubleChannel, default=10961.4736, const [10965.8164,3]}
       Location.Z {type=MovieSceneScriptingDoubleChannel, default=1014.7042, const [1024.7111,3]}
       Rotation.X {type=MovieSceneScriptingDoubleChannel, default=-1741.1594, keys [[52,-1747.9203],[91,-1724.9196],[135,-1724.9196]]}
       Rotation.Y {type=MovieSceneScriptingDoubleChannel, default=-85.3784, const [-85.3784,3]}
       Rotation.Z {type=MovieSceneScriptingDoubleChannel, default=11549.1582, const [11549.1582,3]}
       Scale.X {type=MovieSceneScriptingDoubleChannel, default=1, const [1,3]}
       Scale.Y {type=MovieSceneScriptingDoubleChannel, default=1, const [1,3]}
       Scale.Z {type=MovieSceneScriptingDoubleChannel, default=1, const [1,3]}
      ]
     }]
   }
   {type=MovieSceneSpawnTrack, sections [{type=MovieSceneSpawnSection, channels [None {type=MovieSceneScriptingBoolChannel, default=true}]}]}
  ]
 }
 shot_21 {id=#28, possessed_class=CineCameraActor, tracks []}
 TNTina_Grenade_Box_Gadget2 {
  id=#29
  possessed_class=SkeletalMeshActor
  tracks [
   {
    type=MovieScene3DTransformTrack
    sections [{
      type=MovieScene3DTransformSection
      channels [
       Location.X {type=MovieSceneScriptingDoubleChannel, default=2324}
       Location.Y {type=MovieSceneScriptingDoubleChannel, default=11132}
       Location.Z {type=MovieSceneScriptingDoubleChannel, default=780}
       Rotation.X {type=MovieSceneScriptingDoubleChannel, default=0}
       Rotation.Y {type=MovieSceneScriptingDoubleChannel, default=0}
       Rotation.Z {type=MovieSceneScriptingDoubleChannel, default=-24.9997}
       Scale.X {type=MovieSceneScriptingDoubleChannel, default=0.5584}
       Scale.Y {type=MovieSceneScriptingDoubleChannel, default=0.5584}
       Scale.Z {type=MovieSceneScriptingDoubleChannel, default=0.5584}
      ]
     }]
   }
   {type=MovieSceneSkeletalAnimationTrack, sections []}
   {
    type=MovieSceneVisibilityTrack
    sections [{type=MovieSceneVisibilitySection, channels [None {type=MovieSceneScriptingBoolChannel, default=false}]}]
   }
  ]
 }
 Shot_50_Alt {
  id=#30
  possessed_class=CineCameraActor
  tracks [
   {
    type=MovieScene3DTransformTrack
    sections [{
      type=MovieScene3DTransformSection
      channels [
       Location.X {
        type=MovieSceneScriptingDoubleChannel
        default=2071.0984
        thinned_from=17
        keys [[253,2134.4678],[264,2296.4441],[270,2356.147],[274,2369.3057],[291,2378.6692],[359.125,2379.9402]]
       }
       Location.Y {
        type=MovieSceneScriptingDoubleChannel
        default=10988.2959
        thinned_from=15
        keys [[242,10955.2646],[253,10991.8105],[274,11095.3311],[359.125,11103.874]]
       }
       Location.Z {
        type=MovieSceneScriptingDoubleChannel
        default=807.3429
        thinned_from=17
        keys [[242,812.1912],[253,806.6149],[270,791.3929],[279,788.0003],[281,789.6585],[283,788.1249],[285,789.0836],[359.125,788.3553]]
       }
       Rotation.X {
        type=MovieSceneScriptingDoubleChannel
        default=-5.415
        thinned_from=29
        keys [[242,-1.7184],[253,-5.865],[270,-11.5339],[278,-12.276],[291,-12.2735],[299.2487,-12.3046],[300,-12.3674],[301.2487,-13.439],[302,-11.6351],[303,-13.5132],[304.125,-11.9148],[304.2487,-11.6709],[306,-13.1396],[307.2487,-12.3423],[308.2487,-13.1055],[309.2487,-12.8343],[312.2487,-12.7147],[317,-12.7281],[318,-12.7843],[319.2487,-12.7332],[343.2487,-12.8463],[359.125,-12.8463]]
       }
       Rotation.Y {
        type=MovieSceneScriptingDoubleChannel
        default=20.032
        thinned_from=27
        keys [[242,8.6346],[253,14.422],[264,23.3862],[270,24.4645],[274,24.0116],[279,20.962],[281,18.3698],[282,19.0216],[283,19.5282],[284,18.9687],[285,18.5134],[287,19.0794],[289,18.7517],[294,18.7697],[295,17.8082],[297,19.8589],[298,19.0652],[299,17.9952],[300,18.6961],[301,19.1797],[303,18.3937],[304,18.7878],[306,18.8618],[359.125,18.9133]]
       }
       Rotation.Z {
        type=MovieSceneScriptingDoubleChannel
        default=171.5771
        thinned_from=23
        keys [[242,171.0032],[270,172.4895],[300.125,172.9826],[301.125,172.6867],[302.125,173.1807],[303.125,172.7485],[306.125,172.989],[308.125,172.5032],[312.125,172.9579],[313.125,172.6355],[317.125,173.0697],[359.125,173.2251]]
       }
       Scale.X {type=MovieSceneScriptingDoubleChannel, default=1, const [1,110]}
       Scale.Y {type=MovieSceneScriptingDoubleChannel, default=1, const [1,110]}
       Scale.Z {type=MovieSceneScriptingDoubleChannel, default=1, const [1,110]}
      ]
     }]
   }
   {
    type=MovieSceneCameraShakeTrack
    sections [{type=MovieSceneCameraShakeSection, end_frame=345, start_frame=299}, {type=MovieSceneCameraShakeSection, end_frame=291, start_frame=279}]
   }
   {
    type=MovieSceneSpawnTrack
    sections [{type=MovieSceneSpawnSection, channels [None {type=MovieSceneScriptingBoolChannel, default=true, const [true,1]}]}]
   }
  ]
 }
 Grenades {
  id=#31
  possessed_class=SkeletalMeshActor
  tracks [
   {
    type=MovieScene3DTransformTrack
    sections [{
      type=MovieScene3DTransformSection
      channels [
       Location.X {type=MovieSceneScriptingDoubleChannel, default=0}
       Location.Y {type=MovieSceneScriptingDoubleChannel, default=0}
       Location.Z {type=MovieSceneScriptingDoubleChannel, default=0}
       Rotation.X {type=MovieSceneScriptingDoubleChannel, default=0}
       Rotation.Y {type=MovieSceneScriptingDoubleChannel, default=0}
       Rotation.Z {type=MovieSceneScriptingDoubleChannel, default=0}
       Scale.X {type=MovieSceneScriptingDoubleChannel, default=1}
       Scale.Y {type=MovieSceneScriptingDoubleChannel, default=1}
       Scale.Z {type=MovieSceneScriptingDoubleChannel, default=1}
      ]
     }]
   }
   {
    type=MovieSceneSkeletalAnimationTrack
    sections [
     {type=MovieSceneSkeletalAnimationSection, end_frame=150, start_frame=120}
     {type=MovieSceneSkeletalAnimationSection, end_frame=52, start_frame=1}
     {type=MovieSceneSkeletalAnimationSection, end_frame=120, start_frame=91}
     {type=MovieSceneSkeletalAnimationSection, end_frame=91, start_frame=52}
    ]
   }
   {
    type=MovieSceneSkeletalAnimationTrack
    sections [
     {type=MovieSceneSkeletalAnimationSection, end_frame=185, start_frame=150}
     {type=MovieSceneSkeletalAnimationSection, end_frame=227, start_frame=185}
     {type=MovieSceneSkeletalAnimationSection, end_frame=380, start_frame=252}
    ]
   }
   {type=MovieSceneSpawnTrack, sections [{type=MovieSceneSpawnSection, channels [None {type=MovieSceneScriptingBoolChannel, default=true}]}]}
   {type=MovieSceneSkeletalAnimationTrack, sections []}
   {type=MovieSceneSkeletalAnimationTrack, sections []}
  ]
 }
 S_OilRig_Cart_01 {
  id=#32
  possessed_class=FortStaticMeshActor
  tracks [
   {
    type=MovieScene3DTransformTrack
    sections [{
      type=MovieScene3DTransformSection
      channels [
       Location.X {type=MovieSceneScriptingDoubleChannel, default=2092}
       Location.Y {type=MovieSceneScriptingDoubleChannel, default=11059}
       Location.Z {type=MovieSceneScriptingDoubleChannel, default=780}
       Rotation.X {type=MovieSceneScriptingDoubleChannel, default=0}
       Rotation.Y {type=MovieSceneScriptingDoubleChannel, default=0}
       Rotation.Z {type=MovieSceneScriptingDoubleChannel, default=150.0001}
       Scale.X {type=MovieSceneScriptingDoubleChannel, default=1}
       Scale.Y {type=MovieSceneScriptingDoubleChannel, default=1}
       Scale.Z {type=MovieSceneScriptingDoubleChannel, default=1}
      ]
     }]
   }
   {
    type=MovieSceneVisibilityTrack
    sections [{type=MovieSceneVisibilitySection, channels [None {type=MovieSceneScriptingBoolChannel, default=false}]}]
   }
  ]
 }
 TNTina_Basketball_Hoop_Gadget2 {
  id=#33
  possessed_class=SkeletalMeshActor
  tracks [
   {
    type=MovieScene3DTransformTrack
    sections [{
      type=MovieScene3DTransformSection
      channels [
       Location.X {type=MovieSceneScriptingDoubleChannel, default=2309.7366}
       Location.Y {type=MovieSceneScriptingDoubleChannel, default=11168.6992}
       Location.Z {type=MovieSceneScriptingDoubleChannel, default=886.6577}
       Rotation.X {type=MovieSceneScriptingDoubleChannel, default=0}
       Rotation.Y {type=MovieSceneScriptingDoubleChannel, default=0}
       Rotation.Z {type=MovieSceneScriptingDoubleChannel, default=-180}
       Scale.X {type=MovieSceneScriptingDoubleChannel, default=0.4622}
       Scale.Y {type=MovieSceneScriptingDoubleChannel, default=0.4622}
       Scale.Z {type=MovieSceneScriptingDoubleChannel, default=0.4622}
      ]
     }]
   }
   {type=MovieSceneSkeletalAnimationTrack, sections [{type=MovieSceneSkeletalAnimationSection, end_frame=185, start_frame=150}]}
   {type=MovieSceneSpawnTrack, sections [{type=MovieSceneSpawnSection, channels [None {type=MovieSceneScriptingBoolChannel, default=true}]}]}
  ]
 }
 Prop_BasketballHoops_01_piece4 {
  id=#34
  possessed_class=FortStaticMeshActor
  tracks [
   {
    type=MovieScene3DTransformTrack
    sections [{
      type=MovieScene3DTransformSection
      channels [
       Location.X {type=MovieSceneScriptingDoubleChannel, default=2309.7366}
       Location.Y {type=MovieSceneScriptingDoubleChannel, default=11168.6992}
       Location.Z {type=MovieSceneScriptingDoubleChannel, default=886.6577}
       Rotation.X {type=MovieSceneScriptingDoubleChannel, default=0}
       Rotation.Y {type=MovieSceneScriptingDoubleChannel, default=0}
       Rotation.Z {type=MovieSceneScriptingDoubleChannel, default=-180}
       Scale.X {type=MovieSceneScriptingDoubleChannel, default=0.4622}
       Scale.Y {type=MovieSceneScriptingDoubleChannel, default=0.4622}
       Scale.Z {type=MovieSceneScriptingDoubleChannel, default=0.4622}
      ]
     }]
   }
   {
    type=MovieSceneVisibilityTrack
    sections [{type=MovieSceneVisibilitySection, channels [None {type=MovieSceneScriptingBoolChannel, default=false}]}]
   }
  ]
 }
 S_OilRig_LoadBank_8 {
  id=#35
  possessed_class=FortStaticMeshActor
  tracks [{
    type=MovieScene3DTransformTrack
    sections [{
      type=MovieScene3DTransformSection
      channels [
       Location.X {type=MovieSceneScriptingDoubleChannel, default=2418, keys [[232.0088,2418],[233.0088,2675.0513]]}
       Location.Y {type=MovieSceneScriptingDoubleChannel, default=11221, keys [[232.0088,11221],[233.0088,11221.0117]]}
       Location.Z {type=MovieSceneScriptingDoubleChannel, default=774, const [774,2]}
       Rotation.X {type=MovieSceneScriptingDoubleChannel, default=0, const [0,2]}
       Rotation.Y {type=MovieSceneScriptingDoubleChannel, default=0, const [0,2]}
       Rotation.Z {type=MovieSceneScriptingDoubleChannel, default=-179.9998, const [-179.9998,2]}
       Scale.X {type=MovieSceneScriptingDoubleChannel, default=1.1839, const [1.1839,2]}
       Scale.Y {type=MovieSceneScriptingDoubleChannel, default=1.1839, const [1.1839,2]}
       Scale.Z {type=MovieSceneScriptingDoubleChannel, default=1.1839, const [1.1839,2]}
      ]
     }]
   }]
 }
 CameraComponent {
  id=#36
  parent_id=#27
  possessed_class=CineCameraComponent
  tracks [
   {
    type=MovieSceneFloatTrack
    display_name=Current Focal Length
    sections [{type=MovieSceneFloatSection, channels [None {type=MovieSceneScriptingFloatChannel, default=21.1922}]}]
   }
   {
    type=MovieSceneFloatTrack
    display_name="Manual Focus Distance (Focus Settings)"
    sections [{type=MovieSceneFloatSection, channels [None {type=MovieSceneScriptingFloatChannel, default=100000}]}]
   }
   {
    type=MovieSceneFloatTrack
    display_name=Current Aperture
    sections [{type=MovieSceneFloatSection, channels [None {type=MovieSceneScriptingFloatChannel, default=114.1878}]}]
   }
  ]
 }
 Grenade_pin3 {
  id=#37
  possessed_class=SkeletalMeshActor
  tracks [
   {
    type=MovieScene3DTransformTrack
    sections [{
      type=MovieScene3DTransformSection
      channels [
       Location.X {type=MovieSceneScriptingDoubleChannel, default=0}
       Location.Y {type=MovieSceneScriptingDoubleChannel, default=0}
       Location.Z {type=MovieSceneScriptingDoubleChannel, default=0}
       Rotation.X {type=MovieSceneScriptingDoubleChannel, default=0}
       Rotation.Y {type=MovieSceneScriptingDoubleChannel, default=0}
       Rotation.Z {type=MovieSceneScriptingDoubleChannel, default=0}
       Scale.X {type=MovieSceneScriptingDoubleChannel, default=1}
       Scale.Y {type=MovieSceneScriptingDoubleChannel, default=1}
       Scale.Z {type=MovieSceneScriptingDoubleChannel, default=1}
      ]
     }]
   }
   {type=MovieSceneSpawnTrack, sections [{type=MovieSceneSpawnSection, channels [None {type=MovieSceneScriptingBoolChannel, default=true}]}]}
   {
    type=MovieSceneSkeletalAnimationTrack
    sections [
     {type=MovieSceneSkeletalAnimationSection, end_frame=226, start_frame=185}
     {type=MovieSceneSkeletalAnimationSection, end_frame=380, start_frame=253}
    ]
   }
  ]
 }
 Shot_01 {
  id=#12
  possessed_class=CineCameraActor
  tracks [
   {
    type=MovieScene3DTransformTrack
    sections [{
      type=MovieScene3DTransformSection
      channels [
       Location.X {type=MovieSceneScriptingDoubleChannel, default=2325.6345, keys [[2,2098.6101],[52,2101.9026]]}
       Location.Y {type=MovieSceneScriptingDoubleChannel, default=10938.7744, keys [[2,11005.3516],[52,10969.6289]]}
       Location.Z {type=MovieSceneScriptingDoubleChannel, default=845, keys [[2,869.8086],[52,867.2376]]}
       Rotation.X {type=MovieSceneScriptingDoubleChannel, default=-1.6966, keys [[2,7.4621],[52,4.1901]]}
       Rotation.Y {type=MovieSceneScriptingDoubleChannel, default=12.7354, keys [[2,10.076],[52,11.8032]]}
       Rotation.Z {type=MovieSceneScriptingDoubleChannel, default=95.1559, keys [[2,203.2454],[52,186.1332]]}
       Scale.X {type=MovieSceneScriptingDoubleChannel, default=1, const [1,2]}
       Scale.Y {type=MovieSceneScriptingDoubleChannel, default=1, const [1,2]}
       Scale.Z {type=MovieSceneScriptingDoubleChannel, default=1, const [1,2]}
      ]
     }]
   }
   {type=MovieSceneSpawnTrack, sections [{type=MovieSceneSpawnSection, channels [None {type=MovieSceneScriptingBoolChannel, default=true}]}]}
  ]
 }
 CameraComponent {
  id=#38
  parent_id=#30
  possessed_class=CineCameraComponent
  tracks [
   {
    type=MovieSceneFloatTrack
    display_name=Current Focal Length
    sections [{type=MovieSceneFloatSection, channels [None {type=MovieSceneScriptingFloatChannel, default=15}]}]
   }
   {
    type=MovieSceneFloatTrack
    display_name=FocusSettings.ManualFocusDistance
    sections [{type=MovieSceneFloatSection, channels [None {type=MovieSceneScriptingFloatChannel, default=47.8831, const [47.8831,1]}]}]
   }
   {
    type=MovieSceneFloatTrack
    display_name=Current Aperture
    sections [{type=MovieSceneFloatSection, channels [None {type=MovieSceneScriptingFloatChannel, default=13.9646}]}]
   }
  ]
 }
 CameraComponent {
  id=#39
  parent_id=#40
  possessed_class=CineCameraComponent
  tracks [
   {
    type=MovieSceneFloatTrack
    display_name=Current Focal Length
    sections [{type=MovieSceneFloatSection, channels [None {type=MovieSceneScriptingFloatChannel, default=35}]}]
   }
   {
    type=MovieSceneFloatTrack
    display_name="Manual Focus Distance (Focus Settings)"
    sections [{type=MovieSceneFloatSection, channels [None {type=MovieSceneScriptingFloatChannel, default=100000}]}]
   }
   {
    type=MovieSceneFloatTrack
    display_name=Current Aperture
    sections [{type=MovieSceneFloatSection, channels [None {type=MovieSceneScriptingFloatChannel, default=30}]}]
   }
  ]
 }
 SHOT_0030 {
  id=#40
  possessed_class=CineCameraActor
  tracks [
   {
    type=MovieScene3DTransformTrack
    sections [{
      type=MovieScene3DTransformSection
      channels [
       Location.X {type=MovieSceneScriptingDoubleChannel, default=2366.1135, thinned_from=7, keys [[151.0063,2325.0667],[185.0063,2325.7615]]}
       Location.Y {type=MovieSceneScriptingDoubleChannel, default=10938.3652, thinned_from=7, keys [[151.0088,10940.6035],[185.0063,10938.3652]]}
       Location.Z {type=MovieSceneScriptingDoubleChannel, default=856.2394, const [845,7]}
       Rotation.X {type=MovieSceneScriptingDoubleChannel, default=-1.6966, thinned_from=7, keys [[151.0088,-1.6966],[185.0063,-1.6966]]}
       Rotation.Y {
        type=MovieSceneScriptingDoubleChannel
        default=-12.7957
        keys [[151.0063,13.75],[162,10.3502],[167,8.4496],[170,6.2642],[175,-3.8999],[180,-9.6145],[185,-10.9889]]
       }
       Rotation.Z {
        type=MovieSceneScriptingDoubleChannel
        default=106.6148
        thinned_from=7
        keys [[151.0063,95.9411],[162.0063,94.8161],[167,94.5334],[185,94.5291]]
       }
       Scale.X {type=MovieSceneScriptingDoubleChannel, default=1, const [1,7]}
       Scale.Y {type=MovieSceneScriptingDoubleChannel, default=1, const [1,7]}
       Scale.Z {type=MovieSceneScriptingDoubleChannel, default=1, const [1,7]}
      ]
     }]
   }
   {type=MovieSceneSpawnTrack, sections [{type=MovieSceneSpawnSection, channels [None {type=MovieSceneScriptingBoolChannel, default=true}]}]}
   {type=MovieSceneCameraShakeTrack, sections []}
  ]
 }
 S_OilRig_MovieMemoralbilia_05 {
  id=#41
  possessed_class=FortStaticMeshActor
  tracks [{
    type=MovieScene3DTransformTrack
    sections [{
      type=MovieScene3DTransformSection
      channels [
       Location.X {type=MovieSceneScriptingDoubleChannel, default=2517, keys [[232.0088,2517],[233.0088,2698.5693]]}
       Location.Y {type=MovieSceneScriptingDoubleChannel, default=11136, const [11136,2]}
       Location.Z {type=MovieSceneScriptingDoubleChannel, default=780, const [780,2]}
       Rotation.X {type=MovieSceneScriptingDoubleChannel, default=0, const [0,2]}
       Rotation.Y {type=MovieSceneScriptingDoubleChannel, default=0, const [0,2]}
       Rotation.Z {type=MovieSceneScriptingDoubleChannel, default=-180, const [-180,2]}
       Scale.X {type=MovieSceneScriptingDoubleChannel, default=1, const [1,2]}
       Scale.Y {type=MovieSceneScriptingDoubleChannel, default=1, const [1,2]}
       Scale.Z {type=MovieSceneScriptingDoubleChannel, default=1, const [1,2]}
      ]
     }]
   }]
 }
 CT_NS_TnTina_Intro_Explosion {
  id=#16
  possessed_class=NiagaraActor
  tracks [
   {type=MovieScene3DAttachTrack, sections [{type=MovieScene3DAttachSection, end_frame=364, start_frame=285}]}
   {
    type=MovieScene3DTransformTrack
    sections [{
      type=MovieScene3DTransformSection
      channels [
       Location.X {
        type=MovieSceneScriptingDoubleChannel
        default=0
        thinned_from=5
        keys [[288.9987,2300.9622],[289,4.3582],[387.125,4.3582],[387.1263,2300.9622]]
       }
       Location.Y {
        type=MovieSceneScriptingDoubleChannel
        default=0
        thinned_from=5
        keys [[288.9987,11129.7754],[289,1.9279],[387.125,1.9279],[387.1263,11129.7754]]
       }
       Location.Z {type=MovieSceneScriptingDoubleChannel, default=0, thinned_from=5, keys [[288.9987,806],[289,67.5516],[387.125,67.5516],[387.1263,806]]}
       Rotation.X {type=MovieSceneScriptingDoubleChannel, default=0, const [0,3]}
       Rotation.Y {type=MovieSceneScriptingDoubleChannel, default=0, const [0,3]}
       Rotation.Z {
        type=MovieSceneScriptingDoubleChannel
        default=0
        thinned_from=5
        keys [[288.9987,169.9998],[289,-86.6837],[387.125,-86.6837],[387.1263,169.9998]]
       }
       Scale.X {type=MovieSceneScriptingDoubleChannel, default=1, thinned_from=5, keys [[288.9987,1],[289,2.5],[387.125,2.5],[387.1263,1]]}
       Scale.Y {type=MovieSceneScriptingDoubleChannel, default=1, thinned_from=5, keys [[288.9987,1],[289,2.5],[387.125,2.5],[387.1263,1]]}
       Scale.Z {type=MovieSceneScriptingDoubleChannel, default=1, thinned_from=5, keys [[288.9987,1],[289,2.5],[387.125,2.5],[387.1263,1]]}
      ]
     }]
   }
  ]
 }
 CameraComponent {
  id=#42
  parent_id=#6
  possessed_class=CineCameraComponent
  tracks [
   {
    type=MovieSceneFloatTrack
    display_name=Current Focal Length
    sections [{type=MovieSceneFloatSection, channels [None {type=MovieSceneScriptingFloatChannel, default=20}]}]
   }
   {
    type=MovieSceneFloatTrack
    display_name="Manual Focus Distance (Focus Settings)"
    sections [{type=MovieSceneFloatSection, channels [None {type=MovieSceneScriptingFloatChannel, default=100000}]}]
   }
   {
    type=MovieSceneFloatTrack
    display_name=Current Aperture
    sections [{type=MovieSceneFloatSection, channels [None {type=MovieSceneScriptingFloatChannel, default=8.9301}]}]
   }
  ]
 }
 Camera_Root_Movement {
  id=#24
  possessed_class=Frontend_EventLevel_Camera_C
  tracks [
   {
    type=MovieScene3DTransformTrack
    sections [{
      type=MovieScene3DTransformSection
      channels [
       Location.X {type=MovieSceneScriptingDoubleChannel, default=2378.3931, const [2378.3931,1]}
       Location.Y {type=MovieSceneScriptingDoubleChannel, default=10649.8271, const [10649.8271,1]}
       Location.Z {type=MovieSceneScriptingDoubleChannel, default=876.3943, const [876.3943,1]}
       Rotation.X {type=MovieSceneScriptingDoubleChannel, default=0, const [0,1]}
       Rotation.Y {type=MovieSceneScriptingDoubleChannel, default=0.5995, const [0.5995,1]}
       Rotation.Z {type=MovieSceneScriptingDoubleChannel, default=-228.408, const [-228.408,1]}
       Scale.X {type=MovieSceneScriptingDoubleChannel, default=1}
       Scale.Y {type=MovieSceneScriptingDoubleChannel, default=1}
       Scale.Z {type=MovieSceneScriptingDoubleChannel, default=1}
      ]
     }]
   }
   {
    type=MovieSceneEventTrack
    display_name=Event Stream in Start
    sections [{
      type=MovieSceneEventTriggerSection
      channels [None {type=MovieSceneScriptingEventChannel, const ["<Struct 'MovieSceneEvent' (0x851953b7a80) {}>", 1]}]
     }]
   }
  ]
 }
 CameraComponent {
  id=#43
  parent_id=#18
  possessed_class=CineCameraComponent
  tracks [
   {
    type=MovieSceneFloatTrack
    display_name=Current Focal Length
    sections [{type=MovieSceneFloatSection, channels [None {type=MovieSceneScriptingFloatChannel, default=30, const [30,1]}]}]
   }
   {
    type=MovieSceneFloatTrack
    display_name="Manual Focus Distance (Focus Settings)"
    sections [{
      type=MovieSceneFloatSection
      channels [None {type=MovieSceneScriptingFloatChannel, default=46.667, keys [[185.0063,29.3993],[210,29.3993],[215.0038,46.667]]}]
     }]
   }
   {
    type=MovieSceneFloatTrack
    display_name=Current Aperture
    sections [{type=MovieSceneFloatSection, channels [None {type=MovieSceneScriptingFloatChannel, default=2.6801, const [2.6801,1]}]}]
   }
  ]
 }
 CameraComponent {
  id=#44
  parent_id=#20
  possessed_class=CineCameraComponent
  tracks [
   {
    type=MovieSceneFloatTrack
    display_name=Current Focal Length
    sections [{type=MovieSceneFloatSection, channels [None {type=MovieSceneScriptingFloatChannel, default=20.5463}]}]
   }
   {
    type=MovieSceneFloatTrack
    display_name="Manual Focus Distance (Focus Settings)"
    sections [{type=MovieSceneFloatSection, channels [None {type=MovieSceneScriptingFloatChannel, default=56.6573}]}]
   }
   {
    type=MovieSceneFloatTrack
    display_name=Current Aperture
    sections [{type=MovieSceneFloatSection, channels [None {type=MovieSceneScriptingFloatChannel, default=4.6957}]}]
   }
  ]
 }
]
display_rate {denominator=1, numerator=30}
tick_resolution {denominator=1, numerator=24000}
tracks [
 {
  type=MovieSceneAudioTrack
  display_name=Audio
  sections [{
    type=MovieSceneAudioSection
    end_frame=363
    start_frame=0
    channels [
     Volume {type=MovieSceneScriptingFloatChannel, default=1}
     Pitch {type=MovieSceneScriptingFloatChannel, default=1}
     AttachActor {type=MovieSceneScriptingActorReferenceChannel, default="<Struct 'MovieSceneObjectBindingID' (0x8518f62eb20) {}>"}
    ]
   }]
 }
 {
  type=MovieSceneFadeTrack
  sections [{type=MovieSceneFadeSection, channels [None {type=MovieSceneScriptingFloatChannel, default=1, keys [[1,1],[5,0],[360,0],[364.9987,1]]}]}]
 }
 {
  type=MovieSceneCameraCutTrack
  sections [
   {type=MovieSceneCameraCutSection, end_frame=52, start_frame=-1}
   {type=MovieSceneCameraCutSection, end_frame=1, start_frame=0}
   {type=MovieSceneCameraCutSection, end_frame=52, start_frame=1}
   {type=MovieSceneCameraCutSection, end_frame=91, start_frame=52}
   {type=MovieSceneCameraCutSection, end_frame=120, start_frame=91}
   {type=MovieSceneCameraCutSection, end_frame=150, start_frame=120}
   {type=MovieSceneCameraCutSection, end_frame=185, start_frame=150}
   {type=MovieSceneCameraCutSection, end_frame=229, start_frame=185}
   {type=MovieSceneCameraCutSection, end_frame=253, start_frame=229}
   {type=MovieSceneCameraCutSection, end_frame=371, start_frame=253}
  ]
 }
]
